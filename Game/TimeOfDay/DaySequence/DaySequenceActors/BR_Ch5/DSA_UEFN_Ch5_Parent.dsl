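; Parent Day Sequence Actor for UEFN Chapter 5 time of day, adapting sky and post-processing to scalability settings.
; BeginPlay, visibility, CVar and quality-change events drive updates to volumetric clouds, bloom, motion blur and global shadow, reflection and GI material parameters.
; Cloud materials load async at runtime and blocking in editor, with skybox fallback, while the active bloom event forces standard bloom.
; Custom events also set actor rotation and storm cloud parameters.
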(event EventBeginPlay
  ;[0.1244 0.3385 0.105 0.464 48] Begin Play - Scalability Set Up for in game changes
  (bind _returnvalue (Settings|GetGameUserSettings))
  (EventDispatchers|BindEventtoOnGameUserSettingsUINeedsUpdate _returnvalue (EventDispatchers|CreateEvent :!event UpdateScalabilitySequences))
  (DataDrivenCVar|BindEventtoOnDataDrivenCVarDelegate (EngineSubsystems|GetDataDrivenCVarEngineSubsystem) (delegate Custom|OnDataDrivenCVarDelegate))
  (CallFunction|UpdateComponentVisibility)
  (Sequence|UpdateScalabilitySequences)
  ;;[1 0 0.1306] Remove Later With Code Implementation
  (Development|ExecuteConsoleCommand (Utilities|String|Append "r.Shadow.Virtual.Cache.ForceInvalidateDirectional " (Utilities|String|ToString(Integer) (Variables|VSM|GetInvalidateDirectional)))))

(event Custom|OnDataDrivenCVarDelegate
  (params (CVarName :type String))
  ;[0.1244 0.3385 0.105 0.464 48] Begin Play - Scalability Set Up for in game changes
  (Utilities|FlowControl|Sequence
    (:then_0
      (if (Utilities|String|Equal,CaseInsensitive(String) CVarName "Athena.VolumetricClouds.Quality")
        (CallFunction|EventSetVolCloudMaterial)))
    (:then_1
      (if (Utilities|String|Equal,CaseInsensitive(String) CVarName "Athena.PostProcessLevel")
        (CallFunction|EventSetCustomBloom)))))

(event EventOnTimeOfDayManagerVisibilitySet
  (params (bVisible :type Boolean))
  ;[1 1 1 1 48] Trigger Scalability Settings on Visibility Set
  (if bVisible
    (CallFunction|UpdateComponentVisibility)
    (Sequence|UpdateScalabilitySequences)))

(event Sequence|EventUpdateScalabilitySequences
  ;[1 1 1 1 48] Trigger Scalability Settings on Visibility Set
  (CallFunction|UpdateScalability))

(event EventEndPlay
  (params (EndPlayReason :type EEndPlayReason%20Enum :default "Destroyed"))
  ;[1 0 0.1057] Remove Later with code implementation
  (Development|ExecuteConsoleCommand "r.Shadow.Virtual.Cache.ForceInvalidateDirectional 0"))

(event Scalability|EventOnShadowQualityChanged
  ;[0.1828 0.3073 0.27 1 48] Bind events for scalability
  (CallFunction|SetShadowMPC))

(event Scalability|EventOnEffectsQualityChanged
  ;[0.1828 0.3073 0.27 1 48] Bind events for scalability
  (CallFunction|EventSetVolCloudMaterial))

(event Scalability|EventOnPostProcessQualityChanged
  ;[0.1828 0.3073 0.27 1 48] Bind events for scalability
  (CallFunction|EventSetCustomBloom))

(event Scalability|EventOnReflectionQualityChanged
  ;[0.1828 0.3073 0.27 1 48] Bind events for scalability
  (CallFunction|SetReflectionMPC))

(event Scalability|EventOnGlobalIlluminationQualityChanged
  ;[0.1828 0.3073 0.27 1 48] Bind events for scalability
  (CallFunction|SetLowEndExposure)
  (CallFunction|EventSetCustomBloom)
  (CallFunction|SetGlobalIlluminationMPC))

(event Custom|EventSetCustomBloom
  (bind _bloomconvoutionhighendpostprocess (Variables|Default|GetBloomConvoutionHighEndPostProcess))
  (bind _postprocess (Variables|Default|GetPostProcess))
  ;[0.2656 0.2656 0.2656 1 48] Custom Bloom for UEFN - All forced to Standard Bloom
  (Utilities|FlowControl|Sequence
    (:then_0
      (Rendering|SetVisibility _bloomconvoutionhighendpostprocess false false)
      (Development|SetHiddeninGame _bloomconvoutionhighendpostprocess true false)
      (Fortnite|FortShippingLog self "TODM Scalability: LOWRes Conv Bloom = False"))
    (:then_1
      (Rendering|SetVisibility _postprocess true false)
      (Development|SetHiddeninGame _postprocess false false)
      (Fortnite|FortShippingLog self "TODM Scalability: Standard Bloom = True"))))

(event Custom|EventSetVolCloudMaterial
  (bind _volumetriccloudcomponent (Variables|DayNightCycle|GetVolumetricCloudComponent))
  (bind _volumetriccloudcomponent_2 (Variables|DayNightCycle|GetVolumetricCloudComponent))
  (bind _volumetriccloudcomponent_3 (Variables|DayNightCycle|GetVolumetricCloudComponent))
  (bind _volumetriccloudcomponent_4 (Variables|DayNightCycle|GetVolumetricCloudComponent))
  ;[0.2656 0.2656 0.2656 1 48] Set Volumetric Cloud Visibility and Material
  (if true
    (bind _returnvalue (Development|GetConsoleVariableIntValue "Athena.VolumetricClouds.Quality"))
    (Utilities|FlowControl|Sequence
      (:then_0
        (if (Class|BTechArtGlobalLibrary|UseVolumetricCloudsBR)
          ;;[0.3735 0.5512 1] Volumetric Cloud Material and Visibility
          (Rendering|SetVisibility _volumetriccloudcomponent true false)
          (Development|SetHiddeninGame _volumetriccloudcomponent false false)
          (Rendering|Material|SetScalarParameterValue "/Game/Packages/Fortress_SharedMaterials/GlobalMaterialParameters/FortniteMaterialParameters.FortniteMaterialParameters" "SkyBoxClouds" :!class /Script/Engine.KismetMaterialLibrary)
          (if (>= _returnvalue 100)
            (Variables|Clouds|SetVolumetricCloudMaterialApplied (Variables|Clouds|GetVolumetricCloudMaterialAdvanced))
            (Rendering|SetLayerBottomAltitude _volumetriccloudcomponent_3 0.175)
            (Rendering|SetLayerHeight _volumetriccloudcomponent_3 2.0)
            (Rendering|SetViewSampleCountScale _volumetriccloudcomponent_3 4.0)
            (else
              (Variables|Clouds|SetVolumetricCloudMaterialApplied (Variables|Clouds|GetVolumetricCloudMaterial))
              (Rendering|SetLayerBottomAltitude _volumetriccloudcomponent_4 0.5)
              (Rendering|SetLayerHeight _volumetriccloudcomponent_4 2.0)
              (Rendering|SetViewSampleCountScale _volumetriccloudcomponent_4 2.0)))
          (if (Deprecated|UEFN|IsinEditor)
            (bind _returnvalue_2 (Utilities|LoadAssetBlocking (Variables|Clouds|GetVolumetricCloudMaterialApplied)))
            (bind _asmaterial_interface (Utilities|Casting|CastToMaterialInterface _returnvalue_2))
            (Rendering|SetMaterial (Variables|DayNightCycle|GetVolumetricCloudComponent) _asmaterial_interface)
            (else
              (bind _object (Utilities|AsyncLoadAsset (Variables|Clouds|GetVolumetricCloudMaterialApplied))
                (:Completed
                  (bind _asmaterial_interface_2 (Utilities|Casting|CastToMaterialInterface _object))
                  (Rendering|SetMaterial (Variables|DayNightCycle|GetVolumetricCloudComponent) _asmaterial_interface_2)))))
          (Fortnite|FortShippingLog self "TODM Scalability: Volumetric Clouds = True")
          (else
            (Rendering|SetMaterial (Variables|DayNightCycle|GetVolumetricCloudComponent) "/Game/TimeOfDay/CloudsAndStorm/Materials/MV_EmptyCloudMat.MV_EmptyCloudMat")
            (Rendering|SetVisibility _volumetriccloudcomponent_2 false false)
            (Development|SetHiddeninGame _volumetriccloudcomponent_2 true false)
            (Rendering|Material|SetScalarParameterValue "/Game/Packages/Fortress_SharedMaterials/GlobalMaterialParameters/FortniteMaterialParameters.FortniteMaterialParameters" "SkyBoxClouds" 1.0 :!class /Script/Engine.KismetMaterialLibrary)
            (Fortnite|FortShippingLog self "TODM Scalability: Volumetric Clouds =False"))))
      (:then_1))))

(event Custom|RotateTimeOfDay
  (params (Angle :type Float%20%28double-precision%29))
  ;[1 1 1] Cinematics Event For updating TODM Angle
  (Transformation|SetActorRotation (Math|Rotator|MakeRotator 0.0 0.0 Angle))
  (Fortnite|FortShippingLog self (Utilities|String|Append "DaySequenceActor Angle Updated: " (Utilities|String|ToString(Float) Angle))))

(event Custom|SetCloudsCustomTime
  (params (customTime :type Float%20%28double-precision%29))
  (Rendering|Material|SetScalarParameterValue "/Game/Athena/SafeZone/MPC_SafeZoneClouds.MPC_SafeZoneClouds" "CustomCloudsTime" customTime :!class /Script/Engine.KismetMaterialLibrary))

(event Custom|SetStormPhase
  (params (StormPhase :type Float%20%28double-precision%29))
  (Rendering|Material|SetScalarParameterValue "/Game/Athena/SafeZone/MPC_SafeZoneClouds.MPC_SafeZoneClouds" "StormPhase" StormPhase :!class /Script/Engine.KismetMaterialLibrary))

(event Custom|SetStormPhaseCount
  (params (StormPhaseCount :type Integer :default 5))
  (Rendering|Material|SetScalarParameterValue "/Game/Athena/SafeZone/MPC_SafeZoneClouds.MPC_SafeZoneClouds" "StormPhaseCount" (Math|Conversions|ToFloat(Integer) StormPhaseCount) :!class /Script/Engine.KismetMaterialLibrary))

(fn ComputeCubeMapBlend
  (params (val :type Float%20%28double-precision%29)))

(fn IsExposureCvar
  (params (cvar :type String))
  (return (Math|Boolean|ORBoolean (Utilities|String|Equal,CaseInsensitive(String) cvar "ExposureOverride.Enable") (Utilities|String|Equal,CaseInsensitive(String) cvar "ExposureOverride.DayMin") (Utilities|String|Equal,CaseInsensitive(String) cvar "ExposureOverride.DayMax") (Utilities|String|Equal,CaseInsensitive(String) cvar "ExposureOverride.DayCompensation") (Utilities|String|Equal,CaseInsensitive(String) cvar "ExposureOverride.DaySkyLightLeaking") (Utilities|String|Equal,CaseInsensitive(String) cvar "ExposureOverride.NightMin") (Utilities|String|Equal,CaseInsensitive(String) cvar "ExposureOverride.NightMax") (Utilities|String|Equal,CaseInsensitive(String) cvar "ExposureOverride.NightCompensation") (Utilities|String|Equal,CaseInsensitive(String) cvar "ExposureOverride.NightSkyLightLeaking") (Utilities|String|Equal,CaseInsensitive(String) cvar "ExposureOverride.DayHighlightContrast") (Utilities|String|Equal,CaseInsensitive(String) cvar "ExposureOverride.DayShadowContrast") (Utilities|String|Equal,CaseInsensitive(String) cvar "ExposureOverride.NightHighlightContrast") (Utilities|String|Equal,CaseInsensitive(String) cvar "ExposureOverride.NightShadowContrast") (Utilities|String|Equal,CaseInsensitive(String) cvar "ExposureOverride.DayBlurredLuminanceBlend"))))

(fn SetCustomBloom
  (bind _bloomconvoutionhighendpostprocess (Variables|Default|GetBloomConvoutionHighEndPostProcess))
  (bind _returnvalue (Development|GetConsoleVariableIntValue "Athena.PostProcessLevel"))
  (if (>= _returnvalue 3)
    (Rendering|SetVisibility _bloomconvoutionhighendpostprocess true false)
    (Development|SetHiddeninGame _bloomconvoutionhighendpostprocess false false)
    (bind _returnvalue_2 (Utilities|LoadAssetBlocking (Variables|Bloom|GetConvolution%20Kernel)))
    (bind _astexture_2d (Utilities|Casting|CastToTexture2D _returnvalue_2))
    (bind _structout (Utilities|Struct|SetmembersinPostProcessSettings (Variables|/Script/Engine.PostProcessComponent|GetSettings (Variables|Default|GetBloomConvoutionHighEndPostProcess)) _astexture_2d))
    (Variables|/Script/Engine.PostProcessComponent|SetSettings _structout (Variables|Default|GetBloomConvoutionHighEndPostProcess))
    (Fortnite|FortShippingLog self "TODM Scalability: Convolution Bloom = True")
    (else
      (Rendering|SetVisibility _bloomconvoutionhighendpostprocess false false)
      (Development|SetHiddeninGame _bloomconvoutionhighendpostprocess true false)
      (bind _structout_2 (Utilities|Struct|SetmembersinPostProcessSettings (Variables|/Script/Engine.PostProcessComponent|GetSettings (Variables|Default|GetBloomConvoutionHighEndPostProcess)) ""))
      (Variables|/Script/Engine.PostProcessComponent|SetSettings _structout_2 (Variables|Default|GetBloomConvoutionHighEndPostProcess))
      (Fortnite|FortShippingLog self "TODM Scalability: Convolution Bloom = False"))))

(fn SetGlobalIlluminationMPC
  (bind _returnvalue (Settings|GetGameUserSettings))
  (Rendering|Material|SetScalarParameterValue "/Game/Packages/Fortress_SharedMaterials/GlobalMaterialParameters/FortniteMaterialParameters.FortniteMaterialParameters" "GlobalIlluminationEnabled" (Math|Conversions|ToFloat(Integer) (Math|Integer|Clamp(Integer) (- (Settings|GetGlobalIlluminationQuality _returnvalue) 1) 0 1)) :!class /Script/Engine.KismetMaterialLibrary)
  (Rendering|Material|SetScalarParameterValue "/Game/Packages/Fortress_SharedMaterials/GlobalMaterialParameters/FortniteMaterialParameters.FortniteMaterialParameters" "GlobalIlluminationEpicQuality" (Math|Conversions|ToFloat(Integer) (Math|Integer|Clamp(Integer) (- (Settings|GetGlobalIlluminationQuality _returnvalue) 2) 0 1)) :!class /Script/Engine.KismetMaterialLibrary)
  (Development|PrintString (Utilities|String|Append "GlobalIlluminationDebug " (Utilities|String|ToString(Integer) (Settings|GetGlobalIlluminationQuality _returnvalue)))))

(fn SetLowEndExposure)

(fn SetMotionBlurAmount
  (bind _returnvalue (Development|GetConsoleVariableIntValue "t.MaxFPS"))
  (Variables|/Script/Engine.PostProcessComponent|SetbEnabled (or (> _returnvalue 60) (== _returnvalue 0)) (Variables|Default|Get120fpsMotionBlurPostProcess)))

(fn SetReflectionMPC
  (bind _returnvalue (Settings|GetGameUserSettings))
  (Rendering|Material|SetScalarParameterValue "/Game/Packages/Fortress_SharedMaterials/GlobalMaterialParameters/FortniteMaterialParameters.FortniteMaterialParameters" "ReflectionsEnabled" (Math|Conversions|ToFloat(Integer) (Math|Integer|Clamp(Integer) (Settings|GetReflectionQuality _returnvalue) 0 1)) :!class /Script/Engine.KismetMaterialLibrary))

(fn SetShadowMPC
  (bind _returnvalue (Settings|GetGameUserSettings))
  (Rendering|Material|SetScalarParameterValue "/Game/Packages/Fortress_SharedMaterials/GlobalMaterialParameters/FortniteMaterialParameters.FortniteMaterialParameters" "ShadowsEnabled" (Math|Conversions|ToFloat(Integer) (Math|Integer|Clamp(Integer) (Settings|GetShadowQuality _returnvalue) 0 1)) :!class /Script/Engine.KismetMaterialLibrary))

(fn SetVolumetricCloudMaterial
  (bind _volumetriccloudcomponent (Variables|DayNightCycle|GetVolumetricCloudComponent))
  (bind _returnvalue (Development|GetConsoleVariableIntValue "Athena.VolumetricClouds.Quality"))
  (if (> _returnvalue 0)
    (bind _returnvalue_2 (Utilities|LoadAssetBlocking (Variables|Clouds|GetVolumetricCloudMaterial)))
    (bind _asmaterial_interface (Utilities|Casting|CastToMaterialInterface _returnvalue_2))
    (Rendering|SetMaterial (Variables|DayNightCycle|GetVolumetricCloudComponent) _asmaterial_interface)
    (Rendering|Material|SetScalarParameterValue "/Game/Packages/Fortress_SharedMaterials/GlobalMaterialParameters/FortniteMaterialParameters.FortniteMaterialParameters" "SkyBoxClouds" :!class /Script/Engine.KismetMaterialLibrary)
    (Fortnite|FortShippingLog self "TODM Scalability: Volumetric Clouds = True")
    (else
      (Rendering|SetMaterial (Variables|DayNightCycle|GetVolumetricCloudComponent) "/Game/TimeOfDay/CloudsAndStorm/Materials/MV_EmptyCloudMat.MV_EmptyCloudMat")
      (Rendering|SetVisibility _volumetriccloudcomponent false false)
      (Development|SetHiddeninGame _volumetriccloudcomponent true false)
      (Rendering|Material|SetScalarParameterValue "/Game/Packages/Fortress_SharedMaterials/GlobalMaterialParameters/FortniteMaterialParameters.FortniteMaterialParameters" "SkyBoxClouds" 1.0 :!class /Script/Engine.KismetMaterialLibrary)
      (Fortnite|FortShippingLog self "TODM Scalability: Volumetric Clouds = False"))))

(fn ToggleStormMeshVisibility
  (params (Visibility :type Boolean))
  ;[0 0.7534 1 0.2] Updates the mesh visibility for all actors that care (this is BP_SafeZoneStorm and BP_StormMeshRotator, if they exist)
  (bind _outactors (Actor|GetAllActorswithInterface "/Game/Athena/SafeZone/StormDaySequenceInterface.StormDaySequenceInterface_C"))
  (for _array_element _outactors
    (Storm/DaySequenceInterface|UpdateStormMeshVisibility(Message) _array_element Visibility)))

(fn UpdateComponentVisibility
  ;[1 1 1] Functions that toggle component visibility
  (CallFunction|SetLowEndExposure)
  (CallFunction|EventSetVolCloudMaterial)
  (CallFunction|EventSetCustomBloom)
  (CallFunction|SetShadowMPC)
  (CallFunction|SetReflectionMPC)
  (CallFunction|SetGlobalIlluminationMPC))

(fn UpdateScalability
  (Default|SetMotionBlurAmount)
  (Fortnite|FortShippingLog self (Utilities|String|Append "TODM Scalability: TODM Visibility = " (Utilities|String|ToString(Boolean) (TimeOfDay|IsVisible)))))

(fn ConstructionScript
  (if (Editor|InEditor)
    (CallFunction|EventSetVolCloudMaterial)
    (CallFunction|EventSetCustomBloom)))
